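# Physics asset for secondary motion of the collar and jacket on the male medium Survival Specialist player body.
# Seven collar flaps hang from small MSTR anchor bodies tied to spine_05, each with tight twist and one wide swing axis for hinge-like motion.
# Jacket back and side panels attach to spine_03, as do four two-link front capsule chains.
# Head, spine and upper-arm capsules provide body shapes, joined by free-angle constraints.
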PhysicsAsset /BRCosmetics/Characters/Player/Male/Medium/Bodies/M_MED_SurvivalSpecialist/Meshes/M_MED_Survival_Specialist_Dyn
bodies {
 dyn_collar_01 [
  {shapeType=Box, extentX=10.9747, extentY=6.1347, extentZ=2.0318, center=<4.6778 0 0.201>}
  {shapeType=Box, extentX=11.0947, extentY=6.1347, extentZ=2.0318, center=<4.0417 -4.3769 -0.2456>, rotation=<rot 0 -15 -10>}
  {shapeType=Box, extentX=11.0947, extentY=6.1347, extentZ=2.0318, center=<4.0417 4.3769 -0.2456>, rotation=<rot 0 15.0002 10.0002>}
 ]
 dyn_collar_01_MSTR [{shapeType=Capsule, length=1, radius=1}]
 dyn_collar_01_l [
  {shapeType=Box, extentX=5.2111, extentY=6.6326, extentZ=2, center=<4.5932 3.0841 0.0522>, rotation=<rot 14.9415 33.663 -5.1756>}
  {shapeType=Box, extentX=5.2111, extentY=6.1973, extentZ=2, center=<1.6227 7.538 -0.0827>, rotation=<rot 19.9205 36.8169 5.3193>}
 ]
 dyn_collar_01_l_MSTR [{shapeType=Capsule, length=1, radius=1}]
 dyn_collar_01_r [{shapeType=Box, extentX=5, extentY=9, extentZ=2, center=<4 -4 0>, rotation=<rot 0 -35 5>}]
 dyn_collar_01_r_MSTR [{shapeType=Capsule, length=1, radius=1}]
 dyn_collar_02_l [
  {shapeType=Box, extentX=4.6384, extentY=6.2372, extentZ=1.9548, center=<3.0123 -0.0697 -0.4019>, rotation=<rot 0 -10 -10>}
  {shapeType=Box, extentX=4.82, extentY=5.6249, extentZ=1.9548, center=<3.9267 5.2592 -0.2041>, rotation=<rot 0 -10 5>}
 ]
 dyn_collar_02_l_MSTR [{shapeType=Capsule, length=1, radius=1}]
 dyn_collar_02_r [
  {shapeType=Box, extentX=4.4057, extentY=6.2372, extentZ=1.9548, center=<3.0123 -0.0697 -0.4019>, rotation=<rot 0 10.0005 10.0005>}
  {shapeType=Box, extentX=4.4057, extentY=5.6249, extentZ=1.9548, center=<3.9267 -5.2592 -0.2041>, rotation=<rot 0 10.0031 -5.0007>}
 ]
 dyn_collar_02_r_MSTR [{shapeType=Capsule, length=1, radius=1}]
 dyn_collar_03_l [
  {shapeType=Box, extentX=10.6868, extentY=6, extentZ=2, center=<2.2554 -1.1005 0.0001>, rotation=<rot 0 -15.0003 0>}
  {shapeType=Box, extentX=10.6545, extentY=6.5, extentZ=2, center=<3.0567 3.316 -0.3188>, rotation=<rot 0 -9.0012 9.0015>}
 ]
 dyn_collar_03_l_MSTR [{shapeType=Capsule, length=1, radius=1}]
 dyn_collar_03_r [
  {shapeType=Box, extentX=10.1367, extentY=6, extentZ=2, center=<2.4894 1.1632 0>, rotation=<rot 0 15.0002 5>}
  {shapeType=Box, extentX=10.166, extentY=6.5, extentZ=2, center=<3.2322 -2.6216 0>, rotation=<rot 0 9.0002 -4>}
 ]
 dyn_collar_03_r_MSTR [{shapeType=Capsule, length=1, radius=1}]
 dyn_jacket_bk [{shapeType=Box, extentX=12, extentY=7.5, extentZ=2, center=<6 0 0>}]
 dyn_jacket_fr_l_01 [{shapeType=Capsule, length=4.8505, radius=1.1495, center=<2.5 0 0>, rotation=<rot 90 0 0>}]
 dyn_jacket_fr_l_02 [{shapeType=Capsule, length=4.8505, radius=1.1495, center=<2.5 0 0>, rotation=<rot 90 0 0>}]
 dyn_jacket_fr_r_01 [{shapeType=Capsule, length=4.8505, radius=1.1495, center=<2.5 0 0>, rotation=<rot 90 0 0>}]
 dyn_jacket_fr_r_02 [{shapeType=Capsule, length=4.8505, radius=1.1495, center=<2.5 0 0>, rotation=<rot 90 0 0>}]
 dyn_jacket_l [{shapeType=Box, extentX=9.6809, extentY=7.5, extentZ=2, center=<5.4861 0 0>}]
 dyn_jacket_l_bk [{shapeType=Box, extentX=10, extentY=7.5, extentZ=2, center=<5 0 0>}]
 dyn_jacket_l_fr_01 [{shapeType=Capsule, length=5.8505, radius=1.3495, center=<3 0 0>, rotation=<rot 90 0 0>}]
 dyn_jacket_l_fr_02 [{shapeType=Capsule, length=5.8505, radius=1.3495, center=<3 0 0>, rotation=<rot 90 0 0>}]
 dyn_jacket_r [{shapeType=Box, extentX=9.6809, extentY=7.5, extentZ=2, center=<5.4861 0 0>}]
 dyn_jacket_r_bk [{shapeType=Box, extentX=10, extentY=7.5, extentZ=2, center=<5 0 0>}]
 dyn_jacket_r_fr_01 [{shapeType=Capsule, length=5.8505, radius=1.3495, center=<3 0 0>, rotation=<rot 90 0 0>}]
 dyn_jacket_r_fr_02 [{shapeType=Capsule, length=5.8505, radius=1.3495, center=<3 0 0>, rotation=<rot 90 0 0>}]
 head [{shapeType=Capsule, length=7.1451, radius=7.5803, center=<6.7369 1.615 0>, rotation=<rot -0.0001 4.9983 89.9989>}]
 spine_03 [
  {shapeType=Capsule, length=7.2936, radius=11.4729, center=<0 -0.7035 0>}
  {shapeType=Capsule, length=5.5878, radius=11.3215, center=<-0.0001 1.299 0>}
 ]
 spine_05 [{shapeType=Capsule, length=1, radius=1}]
 upperarm_l [{shapeType=Capsule, length=8, radius=4, center=<7.8 0 0>, rotation=<rot 90 0 0>}]
 upperarm_r [{shapeType=Capsule, length=8.5974, radius=3.5307, center=<-13.4978 -0.0004 0>, rotation=<rot 90 0 0>}]
}
constraints [
 {
  bone1Name=head
  bone2Name=spine_05
  swing1LimitDegrees=45
  swing1Motion=Free
  swing2LimitDegrees=45
  swing2Motion=Free
  twistLimitDegrees=45
  twistMotion=Free
 }
 {bone1Name=dyn_collar_01_l, bone2Name=dyn_collar_01_l_MSTR, swing1LimitDegrees=10, swing2LimitDegrees=90, twistLimitDegrees=2}
 {bone1Name=dyn_collar_02_l, bone2Name=dyn_collar_02_l_MSTR, swing1LimitDegrees=10, swing2LimitDegrees=90, twistLimitDegrees=2}
 {bone1Name=dyn_collar_03_l, bone2Name=dyn_collar_03_l_MSTR, swing1LimitDegrees=5, swing2LimitDegrees=90, twistLimitDegrees=2}
 {bone1Name=dyn_collar_01, bone2Name=dyn_collar_01_MSTR, swing1LimitDegrees=5, swing2LimitDegrees=90, twistLimitDegrees=2}
 {bone1Name=dyn_collar_01_r, bone2Name=dyn_collar_01_r_MSTR, swing1LimitDegrees=10, swing2LimitDegrees=90, twistLimitDegrees=2}
 {bone1Name=dyn_collar_02_r, bone2Name=dyn_collar_02_r_MSTR, swing1LimitDegrees=10, swing2LimitDegrees=90, twistLimitDegrees=2}
 {bone1Name=dyn_collar_03_r, bone2Name=dyn_collar_03_r_MSTR, swing1LimitDegrees=5, swing2LimitDegrees=90, twistLimitDegrees=2}
 {
  bone1Name=dyn_collar_01_l_MSTR
  bone2Name=spine_05
  swing1LimitDegrees=45
  swing1Motion=Free
  swing2LimitDegrees=45
  swing2Motion=Free
  twistLimitDegrees=45
  twistMotion=Free
 }
 {
  bone1Name=dyn_collar_02_l_MSTR
  bone2Name=spine_05
  swing1LimitDegrees=45
  swing1Motion=Free
  swing2LimitDegrees=45
  swing2Motion=Free
  twistLimitDegrees=45
  twistMotion=Free
 }
 {
  bone1Name=dyn_collar_03_l_MSTR
  bone2Name=spine_05
  swing1LimitDegrees=45
  swing1Motion=Free
  swing2LimitDegrees=45
  swing2Motion=Free
  twistLimitDegrees=45
  twistMotion=Free
 }
 {
  bone1Name=dyn_collar_01_MSTR
  bone2Name=spine_05
  swing1LimitDegrees=45
  swing1Motion=Free
  swing2LimitDegrees=45
  swing2Motion=Free
  twistLimitDegrees=45
  twistMotion=Free
 }
 {
  bone1Name=dyn_collar_01_r_MSTR
  bone2Name=spine_05
  swing1LimitDegrees=45
  swing1Motion=Free
  swing2LimitDegrees=45
  swing2Motion=Free
  twistLimitDegrees=45
  twistMotion=Free
 }
 {
  bone1Name=dyn_collar_02_r_MSTR
  bone2Name=spine_05
  swing1LimitDegrees=45
  swing1Motion=Free
  swing2LimitDegrees=45
  swing2Motion=Free
  twistLimitDegrees=45
  twistMotion=Free
 }
 {
  bone1Name=dyn_collar_03_r_MSTR
  bone2Name=spine_05
  swing1LimitDegrees=45
  swing1Motion=Free
  swing2LimitDegrees=45
  swing2Motion=Free
  twistLimitDegrees=45
  twistMotion=Free
 }
 {bone1Name=dyn_jacket_fr_l_02, bone2Name=dyn_jacket_fr_l_01, swing1LimitDegrees=5, swing2LimitDegrees=45, twistLimitDegrees=5}
 {bone1Name=dyn_jacket_l_fr_02, bone2Name=dyn_jacket_l_fr_01, swing1LimitDegrees=5, swing2LimitDegrees=45, twistLimitDegrees=5}
 {bone1Name=dyn_jacket_fr_r_02, bone2Name=dyn_jacket_fr_r_01, swing1LimitDegrees=5, swing2LimitDegrees=45, twistLimitDegrees=5}
 {bone1Name=dyn_jacket_r_fr_02, bone2Name=dyn_jacket_r_fr_01, swing1LimitDegrees=5, swing2LimitDegrees=45, twistLimitDegrees=5}
 {
  bone1Name=spine_05
  bone2Name=spine_03
  swing1LimitDegrees=45
  swing1Motion=Free
  swing2LimitDegrees=45
  swing2Motion=Free
  twistLimitDegrees=45
  twistMotion=Free
 }
 {bone1Name=dyn_jacket_bk, bone2Name=spine_03, swing1LimitDegrees=5, swing2LimitDegrees=90, twistLimitDegrees=5}
 {bone1Name=dyn_jacket_l_bk, bone2Name=spine_03, swing1LimitDegrees=5, swing2LimitDegrees=90, twistLimitDegrees=5}
 {bone1Name=dyn_jacket_l, bone2Name=spine_03, swing1LimitDegrees=5, swing2LimitDegrees=90, twistLimitDegrees=5}
 {bone1Name=dyn_jacket_fr_l_01, bone2Name=spine_03, swing1LimitDegrees=5, swing2LimitDegrees=45, twistLimitDegrees=5}
 {bone1Name=dyn_jacket_l_fr_01, bone2Name=spine_03, swing1LimitDegrees=5, swing2LimitDegrees=45, twistLimitDegrees=5}
 {bone1Name=dyn_jacket_r_bk, bone2Name=spine_03, swing1LimitDegrees=5, swing2LimitDegrees=90, twistLimitDegrees=5}
 {bone1Name=dyn_jacket_r, bone2Name=spine_03, swing1LimitDegrees=5, swing2LimitDegrees=90, twistLimitDegrees=5}
 {bone1Name=dyn_jacket_fr_r_01, bone2Name=spine_03, swing1LimitDegrees=5, swing2LimitDegrees=45, twistLimitDegrees=5}
 {bone1Name=dyn_jacket_r_fr_01, bone2Name=spine_03, swing1LimitDegrees=5, swing2LimitDegrees=45, twistLimitDegrees=5}
 {
  bone1Name=upperarm_r
  bone2Name=spine_05
  swing1LimitDegrees=45
  swing1Motion=Free
  swing2LimitDegrees=45
  swing2Motion=Free
  twistLimitDegrees=45
  twistMotion=Free
 }
 {
  bone1Name=upperarm_l
  bone2Name=spine_05
  swing1LimitDegrees=45
  swing1Motion=Free
  swing2LimitDegrees=45
  swing2Motion=Free
  twistLimitDegrees=45
  twistMotion=Free
 }
]
